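# Post-process style AnimBP for The Origin character body, correcting the dynamic shoulder armor bones on a pose copied from the attached parent mesh.
# Normally, constraints, look-ats and additive bone transforms follow the arms, weighted by upper-arm twist, movement state and weapon flags.
# While skydiving, a separate chain takes over with clavicle constraints, parachute offsets, a Control Rig and look-ats.
# The update event derives twist alphas and reads weapon type from FortPlayerAnimInstance.
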AnimBlueprint /BRCosmetics/Characters/Player/Male/Medium/Bodies/M_MED_The_Origin/Meshes/M_MED_The_Origin_AnimBP
target_skeleton=&/M_MED_The_Origin_Skeleton
graphs [
 AnimGraph {nodes [
   n1 = "Constraint - dyn_shoulder_armor_l (P)"(Alpha=1, ConstraintWeights_0=1) {
    alphaBoolBlend {blendOption=Linear}
    boneToModify {boneName=dyn_shoulder_armor_l}
    constraintSetup [{targetBone {boneName=clavicle_l}}]
    constraintWeights [1]
   }
   "Constraint - dyn_shoulder_armor_r (P)"(ComponentPose=@n1, Alpha=1, ConstraintWeights_0=1) {
    alphaBoolBlend {blendOption=Linear}
    boneToModify {boneName=dyn_shoulder_armor_r}
    constraintSetup [{targetBone {boneName=clavicle_r}}]
    constraintWeights [1]
   }
   n2 = Get Rifle() -> (Rifle:bool)
   n3 = "Constraint - dyn_shoulder_armor_r (P)"!(
    ComponentPose=Local To Component(LocalPose=Copy Pose From Mesh() {bUseAttachedParent=true})
    bAlphaBoolEnabled:bool=OR Boolean(A=@n2, B=Get pickaxe() -> (pickaxe:bool))
    ConstraintWeights_0=1
   ) -> (Pose) {
    alphaInputType=Bool
    alphaBoolBlend {blendInTime=0.1, blendOption=Linear, blendOutTime=0.1}
    boneToModify {boneName=dyn_shoulder_armor_r}
    constraintSetup [{targetBone {boneName=upperarm_twist_01_r}}]
    constraintWeights [1]
   }
   n4 = "Constraint - dyn_shoulder_armor_l (P)"!(ComponentPose=@n3, bAlphaBoolEnabled:bool=@n2, ConstraintWeights_0=1) -> (Pose) {
    alphaInputType=Bool
    alphaBoolBlend {blendInTime=0.1, blendOption=Linear, blendOutTime=0.1}
    boneToModify {boneName=dyn_shoulder_armor_l}
    constraintSetup [{targetBone {boneName=upperarm_twist_01_l}}]
    constraintWeights [1]
   }
   n5 = OR Boolean(A=Get bIsJumping(), B=Get bIsFalling())
   n6 = "Constraint - dyn_shoulder_armor_l (P)"(ComponentPose=@n4, bAlphaBoolEnabled=@n5, ConstraintWeights_0=1) {
    alphaInputType=Bool
    alphaBoolBlend {blendInTime=0.5, blendOption=Linear, blendOutTime=0.5}
    boneToModify {boneName=dyn_shoulder_armor_l}
    constraintSetup [{targetBone {boneName=upperarm_twist_01_l}}]
    constraintWeights [1]
   }
   n7 = "Constraint - dyn_shoulder_armor_r (P)"(ComponentPose=@n6, bAlphaBoolEnabled=@n5, ConstraintWeights_0=1) {
    alphaInputType=Bool
    alphaBoolBlend {blendInTime=0.5, blendOption=Linear, blendOutTime=0.5}
    boneToModify {boneName=dyn_shoulder_armor_r}
    constraintSetup [{targetBone {boneName=upperarm_twist_01_r}}]
    constraintWeights [1]
   }
   n8 = Get Arm L rot() -> (Arm L rot:double)
   n9 = Look At - Bone: dyn_shoulder_armor_l(ComponentPose=@n7, Alpha=@n8) {
    alphaBoolBlend {blendOption=Linear}
    boneToModify {boneName=dyn_shoulder_armor_l}
    lookAtLocation {x=0, z=7}
    lookAtTarget {boneReference {boneName=lowerarm_l}}
    lookAt_Axis {axis=<1 0>}
   }
   n10 = Get Arm R rot() -> (Arm R rot:double)
   n11 = Look At - Bone: dyn_shoulder_armor_r(ComponentPose=@n9, Alpha=@n10) {
    alphaBoolBlend {blendOption=Linear}
    boneToModify {boneName=dyn_shoulder_armor_r}
    lookAtLocation {x=0, z=-5}
    lookAtTarget {boneReference {boneName=lowerarm_r}}
    lookAt_Axis {axis=<-1 0>}
   }
   n12 = "Transform (Modify) Bone - Bone: dyn_shoulder_armor_l"(ComponentPose=@n11, Alpha=@n8, Translation="2,-1,1", Rotation="-10.000,0.000,-0.000", Scale="1,0.9,0.9") {
    rotationMode=BMM_Additive
    rotationSpace=BCS_BoneSpace
    scaleMode=BMM_Additive
    scaleSpace=BCS_BoneSpace
    translationMode=BMM_Additive
    translationSpace=BCS_BoneSpace
    alphaBoolBlend {blendOption=Linear}
    boneToModify {boneName=dyn_shoulder_armor_l}
    rotation=<rot -10 0.0001 -0.0001>
    scale {y=0.9, z=0.9}
    translation=<2 -1 1>
   }
   n13 = "Transform (Modify) Bone - Bone: dyn_shoulder_armor_r"(ComponentPose=@n12, Alpha=@n10, Translation="0,0,0", Rotation="5.000,0.000,-15", Scale="0.9,1,0.85") {
    rotationMode=BMM_Additive
    rotationSpace=BCS_BoneSpace
    scaleMode=BMM_Additive
    scaleSpace=BCS_BoneSpace
    translationMode=BMM_Additive
    translationSpace=BCS_BoneSpace
    alphaBoolBlend {blendOption=Linear}
    boneToModify {boneName=dyn_shoulder_armor_r}
    rotation {pitch=5, roll=-15}
    scale {x=0.9, z=0.85}
   }
   n14 = "Transform (Modify) Bone - Bone: dyn_shoulder_armor_r"!(
    ComponentPose=@n13
    bAlphaBoolEnabled:bool=Get shoulder corrective() -> (shoulder corrective:bool)
    Translation="0,0,0"
    Rotation="5,0,-20"
    Scale="1,1,1"
   ) -> (Pose) {
    alphaInputType=Bool
    rotationMode=BMM_Additive
    rotationSpace=BCS_BoneSpace
    scaleMode=BMM_Additive
    scaleSpace=BCS_BoneSpace
    translationMode=BMM_Additive
    translationSpace=BCS_BoneSpace
    alphaBoolBlend {blendInTime=0.1, blendOption=Linear, blendOutTime=0.4}
    alphaScaleBiasClamp {bInterpResult=true, interpSpeedDecreasing=6, interpSpeedIncreasing=30}
    boneToModify {boneName=dyn_shoulder_armor_r}
    rotation {pitch=5, roll=-20}
   }
   n15 = "Transform (Modify) Bone - Bone: dyn_shoulder_armor_r"!(ComponentPose=@n14, Alpha:float=Get Arm R rot_2() -> (Arm R rot_2:double), Translation="0,1.5,-0.5", Rotation="5,-5,0", Scale="1,0.95,1") -> (Pose) {
    rotationMode=BMM_Additive
    rotationSpace=BCS_BoneSpace
    scaleMode=BMM_Additive
    scaleSpace=BCS_BoneSpace
    translationMode=BMM_Additive
    translationSpace=BCS_BoneSpace
    alphaBoolBlend {blendInTime=0.1, blendOption=Linear, blendOutTime=0.4}
    alphaScaleBiasClamp {bInterpResult=true, interpSpeedDecreasing=6, interpSpeedIncreasing=30}
    boneToModify {boneName=dyn_shoulder_armor_r}
    rotation {pitch=5, yaw=-5}
    scale {y=0.95}
    translation {y=1.5, z=-0.5}
   }
   n16 = "Transform (Modify) Bone - Bone: dyn_shoulder_armor_r"(ComponentPose=@n15, Alpha=Get Arm R rot_3() -> (Arm R rot_3:double), Translation="0,0.5,-0.5", Rotation="0,-5,0", Scale="1,1,1") {
    rotationMode=BMM_Additive
    rotationSpace=BCS_BoneSpace
    scaleMode=BMM_Additive
    scaleSpace=BCS_BoneSpace
    translationMode=BMM_Additive
    translationSpace=BCS_BoneSpace
    alphaBoolBlend {blendInTime=0.1, blendOption=Linear, blendOutTime=0.4}
    alphaScaleBiasClamp {bInterpResult=true, interpSpeedDecreasing=6, interpSpeedIncreasing=30}
    boneToModify {boneName=dyn_shoulder_armor_r}
    rotation {yaw=-5}
    translation {y=0.5, z=-0.5}
   }
   n17 = "Transform (Modify) Bone - Bone: dyn_shoulder_armor_l"!(ComponentPose=@n16, bAlphaBoolEnabled:bool=True, Translation="0.000,0.6,0.5", Rotation="-0.000,0.000,-7", Scale="1,1,1") -> (Pose) {
    alphaInputType=Bool
    rotationMode=BMM_Additive
    rotationSpace=BCS_BoneSpace
    scaleMode=BMM_Additive
    scaleSpace=BCS_BoneSpace
    translationMode=BMM_Additive
    translationSpace=BCS_BoneSpace
    alphaBoolBlend {blendInTime=0.1, blendOption=Linear, blendOutTime=0.1}
    boneToModify {boneName=dyn_shoulder_armor_l}
    rotation {roll=-7}
    translation {y=0.6, z=0.5}
   }
   n18 = "Constraint - dyn_shoulder_armor_l (P)"!(ComponentPose=Local To Component(LocalPose=Copy Pose From Mesh() {bUseAttachedParent=true}), Alpha:float=1, ConstraintWeights_0=1) -> (Pose) {
    alphaBoolBlend {blendOption=Linear}
    boneToModify {boneName=dyn_shoulder_armor_l}
    constraintSetup [{targetBone {boneName=clavicle_l}}]
    constraintWeights [1]
   }
   n19 = "Constraint - dyn_shoulder_armor_r (P)"!(ComponentPose=@n18, Alpha:float=1, ConstraintWeights_0=1) -> (Pose) {
    alphaBoolBlend {blendOption=Linear}
    boneToModify {boneName=dyn_shoulder_armor_r}
    constraintSetup [{targetBone {boneName=clavicle_r}}]
    constraintWeights [1]
   }
   n20 = "Transform (Modify) Bone - Bone: dyn_shoulder_armor_l"!(ComponentPose=@n19, Alpha:float=1, Translation="2,0,2", Rotation="0,0,0", Scale="1,1,1") -> (Pose) {
    rotationMode=BMM_Additive
    rotationSpace=BCS_BoneSpace
    translationMode=BMM_Additive
    translationSpace=BCS_BoneSpace
    alphaBoolBlend {blendOption=Linear}
    boneToModify {boneName=dyn_shoulder_armor_l}
    translation {x=2, z=2}
   }
   n21 = "Transform (Modify) Bone - Bone: dyn_shoulder_armor_r"(ComponentPose=@n20, Alpha=1, Translation="-2,0,-2", Rotation="0,0,0", Scale="1,1,1") {
    rotationMode=BMM_Additive
    rotationSpace=BCS_BoneSpace
    translationMode=BMM_Additive
    translationSpace=BCS_BoneSpace
    alphaBoolBlend {blendOption=Linear}
    boneToModify {boneName=dyn_shoulder_armor_r}
    translation {x=-2, z=-2}
   }
   n22 = Get bIsParachuteOpened()
   n23 = "Constraint - dyn_shoulder_armor_r (P)"!(ComponentPose=@n21, bAlphaBoolEnabled:bool=@n22, ConstraintWeights_0=1) -> (Pose) {
    alphaInputType=Bool
    alphaBoolBlend {blendInTime=0.1, blendOption=Linear, blendOutTime=0.1}
    boneToModify {boneName=dyn_shoulder_armor_r}
    constraintSetup [{targetBone {boneName=upperarm_twist_01_r}}]
    constraintWeights [1]
   }
   n24 = "Transform (Modify) Bone - Bone: dyn_shoulder_armor_r"!(ComponentPose=@n23, bAlphaBoolEnabled:bool=@n22, Translation="-5,0,0", Rotation="0,0,0", Scale="1,1,1") -> (Pose) {
    alphaInputType=Bool
    rotationMode=BMM_Additive
    rotationSpace=BCS_BoneSpace
    translationMode=BMM_Additive
    translationSpace=BCS_BoneSpace
    alphaBoolBlend {blendInTime=0.1, blendOption=Linear, blendOutTime=0.1}
    boneToModify {boneName=dyn_shoulder_armor_r}
    translation {x=-5}
   }
   n25 = Control Rig(Source=Component To Local(ComponentPose=@n24)) {
    alphaBoolBlend {blendOption=Linear}
    controlRigAssetReference {blueprintRigClass=&/M_MED_The_Origin_CtrlRig.M_MED_The_Origin_CtrlRig_C}
   }
   n26 = Look At - Bone: dyn_shoulder_armor_l(ComponentPose=Local To Component(LocalPose=@n25), Alpha=1) {
    alphaBoolBlend {blendOption=Linear}
    boneToModify {boneName=dyn_shoulder_armor_l}
    lookAtLocation {x=0}
    lookAtTarget {boneReference {boneName=lowerarm_l}}
    lookAt_Axis {axis=<1 0>}
   }
   n27 = Look At - Bone: dyn_shoulder_armor_r(ComponentPose=@n26, Alpha=1) {
    alphaBoolBlend {blendOption=Linear}
    boneToModify {boneName=dyn_shoulder_armor_r}
    lookAtLocation {x=0}
    lookAtTarget {boneReference {boneName=lowerarm_r}}
    lookAt_Axis {axis=<-1 0>}
   }
   Output Pose(Result=Blend Poses by bool(BlendPose_0=Component To Local(ComponentPose=@n27), BlendPose_1=Component To Local(ComponentPose=@n17), bActiveValue=Get bIsSkyDiving()) {blendType=Linear}) {layerGroup=None}
  ]}
 EventGraph {nodes [
   Event BlueprintUpdateAnimation()
   > n1 = Sequence!(execute) -> (then_0, then_1, then_2:Exec, then_3:Exec, then_4:Exec, then_5:Exec)
   n2 = Get Skel_Mesh()
   > Set Arm L rot(Arm L rot:double=GetAlphaFromReferenceTransformDelta(SkeletalMeshComponent=@n2, BoneName=upperarm_l, BaseName=spine_05, InputRange=(X=40,Y=90), TwistAxis="1,0,0")) -> (Output_Get:double)
   n3 = GetAlphaFromReferenceTransformDelta(SkeletalMeshComponent=@n2, BoneName=upperarm_r, BaseName=spine_05, InputRange=(X=60,Y=120), TwistAxis="1,0,0")
   > Set Arm R rot(Arm R rot:double=MapRangeClamped(Value="Absolute (Float)"(A=@n3.RawOutput), InRangeA=60, InRangeB=120)) -> (Output_Get:double)
   n4 = Branch(
    execute=@n1.then_1
    Condition=OR Boolean(A=Get bIsRelaxedLevel1() -> (bIsRelaxedLevel1:bool), B=Get bIsRelaxedLevel2() -> (bIsRelaxedLevel2:bool))
   )
   Set shoulder corrective(execute=[@n5.else, @n4.then], shoulder corrective:bool=false) -> (Output_Get:bool)
   n5 = Branch(execute=@n4.else, Condition=NOT Boolean(A=Get bIsSkyDiving()))
   Set shoulder corrective(execute=[@n5.then, @n6.then], shoulder corrective=true)
   n6 = Branch(execute=@n1.then_3, Condition=Get bIsTargeting())
   n7 = Branch(execute=@n1.then_4, Condition=Get shoulder corrective())
   Set Arm R rot_2(execute=@n7.else, Arm R rot_2:double=0) -> (Output_Get:double)
   > Set Arm R rot_3(Arm R rot_3:double=0) -> (Output_Get:double)
   n8 = Cast To FortPlayerAnimInstance(execute=@n1.then_5, Object=GetAnimInstance(self=GetOwnerSkeletalMeshComponent()))
   > Set Rifle(Rifle:bool=false) -> (Output_Get:bool)
   > Set pistol(pistol:bool=false) -> (Output_Get:bool)
   > Set pickaxe(pickaxe:bool=false) -> (Output_Get:bool)
   > n9 = Switch on EFortWeaponCoreAnimation!(execute, Selection:EFortWeaponCoreAnimation Enum=Get WeaponCoreAnim(self=@n8)) -> (
    Melee:Exec
    Pistol:Exec
    Shotgun:Exec
    PaperBlueprint:Exec
    Rifle:Exec
    MeleeOneHand:Exec
    MachinePistol:Exec
    RocketLauncher:Exec
    GrenadeLauncher:Exec
    GoingCommando:Exec
    AssaultRifle:Exec
    TacticalShotgun:Exec
    SniperRifle:Exec
    TrapPlacement:Exec
    ShoulderLauncher:Exec
    AbilityDecoTool:Exec
    Crossbow:Exec
    C4:Exec
    RemoteControl:Exec
    DualWield:Exec
    AR_BullPup:Exec
    AR_ForwardGrip:Exec
    MedPackPaddles:Exec
    SMG_P90:Exec
    AR_DrumGun:Exec
    Consumable_Small:Exec
    Consumable_Large:Exec
    Balloon:Exec
    MountedTurret:Exec
    CreativeTool:Exec
    ExplosiveBow:Exec
    AshtonIndigo:Exec
    AshtonChicago:Exec
    MeleeDualWield:Exec
    Unarmed:Exec
   )
   Branch(execute=[@n9.AssaultRifle, @n9.Crossbow, @n9.GrenadeLauncher, @n9.Rifle, @n9.RocketLauncher, @n9.Shotgun, @n9.SniperRifle, @n9.TacticalShotgun])
   > Set Rifle(Rifle=true)
   > n10 = Branch(Condition=Get shoulder corrective())
   > Set Arm R rot_3(Arm R rot_3=GetAlphaFromReferenceTransformDelta(SkeletalMeshComponent=Get Skel_Mesh(), BoneName=upperarm_r, BaseName=spine_05, InputRange=(X=120,Y=40), TwistAxis="1,0,0"))
   > Set Arm R rot_2(Arm R rot_2=0)
   Set Arm R rot_3(execute=@n10.else, Arm R rot_3=0)
   Branch(execute=[@n9.MachinePistol, @n9.Pistol])
   > Set pistol(pistol=true)
   > n11 = Branch(Condition=Get shoulder corrective())
   > Set Arm R rot_2(Arm R rot_2=GetAlphaFromReferenceTransformDelta(SkeletalMeshComponent=Get Skel_Mesh(), BoneName=upperarm_r, BaseName=spine_05, InputRange=(X=120,Y=40), TwistAxis="1,0,0"))
   > Set Arm R rot_3(Arm R rot_3=0)
   Set Arm R rot_2(execute=@n11.else, Arm R rot_2=0)
   Set shoulder corrective(execute=@n9.Melee, shoulder corrective=false)
   > Set Arm R rot_2(Arm R rot_2=0)
   > Set Arm R rot_3(Arm R rot_3=0)
   > Set pickaxe(pickaxe=true)
   Event BlueprintInitializeAnimation()
   > Set Skel_Mesh(Skel_Mesh=GetOwnerSkeletalMeshComponent())
   PrintString!(execute, InString=no, bPrintToScreen, bPrintToLog, TextColor, Duration, Key) -> (then)
   PrintString(InString=Hello)
  ]}
]
variables [
 Arm L rot {type=number}
 Arm R rot {type=number}
 Arm R rot_2 {type=number}
 Arm R rot_3 {type=number}
 pickaxe {type=boolean}
 pistol {type=boolean}
 Rifle {type=boolean}
 shoulder corrective {type=boolean}
 Skel_Mesh {type=SkeletalMeshComponent}
]
