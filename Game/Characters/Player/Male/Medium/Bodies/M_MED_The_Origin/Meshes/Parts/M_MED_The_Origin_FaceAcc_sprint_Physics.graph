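# Physics asset for the sprint face accessory of the medium male player body The Origin, covering two dangling "dyn_floaty" pieces and the upper body around them.
# Capsules cover head, upper spine, clavicles and upper arms, linked back to spine_05 by constraints with free swing and twist.
# Each long floaty capsule is angularly locked to its tiny MSTR root capsule, and no constraint links either root to the body.
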PhysicsAsset /Game/Characters/Player/Male/Medium/Bodies/M_MED_The_Origin/Meshes/Parts/M_MED_The_Origin_FaceAcc_sprint_Physics
bodies {
 clavicle_l [{shapeType=Capsule, length=10, radius=9.005, shapeName=clavicle_l_capsule, center=<9 0 0>, rotation=<rot 90 0 0>}]
 clavicle_r [{shapeType=Capsule, length=10, radius=9.005, shapeName=clavicle_r_capsule, center=<-9 0 0>, rotation=<rot 90 0 0>}]
 dyn_floaty_l [{shapeType=Capsule, length=23.6727, radius=2.7271, shapeName=dyn_floaty_l_capsule, rotation=<rot 4.4116 -90 36.6446>}]
 dyn_floaty_l_MSTR [{shapeType=Capsule, length=0.505, radius=0.505, shapeName=dyn_floaty_l_MSTR_capsule, rotation=<rot 0 -90 0>}]
 dyn_floaty_r [{shapeType=Capsule, length=23.4227, radius=2.9771, shapeName=dyn_floaty_r_capsule, rotation=<rot -4.4116 -90 36.6446>}]
 dyn_floaty_r_MSTR [{shapeType=Capsule, length=0.505, radius=0.505, shapeName=dyn_floaty_r_MSTR_capsule, rotation=<rot 0 -90 0>}]
 head [{shapeType=Capsule, length=9.1195, radius=6.399, shapeName=head_capsule, center=<1.6 3.5892 0.0007>, rotation=<rot -0.009 -90.66 -89.9924>}]
 spine_05 [{
   shapeType=Capsule
   length=13.985
   radius=14.5155
   shapeName=spine_05_capsule
   center=<-5.6123 2.1934 0.0593>
   rotation=<rot 0 -146.6349 0.3112>
  }]
 upperarm_l [{shapeType=Capsule, length=2.5, radius=5.95, shapeName=upperarm_l_capsule, center=<4 0 5>, rotation=<rot 90 0 0>}]
 upperarm_r [{shapeType=Capsule, length=3, radius=6.005, shapeName=upperarm_r_capsule, center=<-4.2444 0 -4.7976>, rotation=<rot 90 0 0>}]
}
constraints [
 {
  bone1Name=dyn_floaty_r
  bone2Name=dyn_floaty_r_MSTR
  swing1LimitDegrees=5
  swing1Motion=Locked
  swing2LimitDegrees=20
  swing2Motion=Locked
  twistLimitDegrees=15
  twistMotion=Locked
 }
 {
  bone1Name=dyn_floaty_l
  bone2Name=dyn_floaty_l_MSTR
  swing1LimitDegrees=5
  swing1Motion=Locked
  swing2LimitDegrees=20
  swing2Motion=Locked
  twistLimitDegrees=15
  twistMotion=Locked
 }
 {
  bone1Name=upperarm_l
  bone2Name=clavicle_l
  swing1LimitDegrees=45
  swing1Motion=Free
  swing2LimitDegrees=45
  swing2Motion=Free
  twistLimitDegrees=45
  twistMotion=Free
 }
 {
  bone1Name=upperarm_r
  bone2Name=clavicle_r
  swing1LimitDegrees=45
  swing1Motion=Free
  swing2LimitDegrees=45
  swing2Motion=Free
  twistLimitDegrees=45
  twistMotion=Free
 }
 {
  bone1Name=clavicle_l
  bone2Name=spine_05
  swing1LimitDegrees=45
  swing1Motion=Free
  swing2LimitDegrees=45
  swing2Motion=Free
  twistLimitDegrees=45
  twistMotion=Free
 }
 {
  bone1Name=clavicle_r
  bone2Name=spine_05
  swing1LimitDegrees=45
  swing1Motion=Free
  swing2LimitDegrees=45
  swing2Motion=Free
  twistLimitDegrees=45
  twistMotion=Free
 }
 {
  bone1Name=head
  bone2Name=spine_05
  swing1LimitDegrees=45
  swing1Motion=Free
  swing2LimitDegrees=45
  swing2Motion=Free
  twistLimitDegrees=45
  twistMotion=Free
 }
]
