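# Animation Blueprint for The Origin outfit's face accessory, animating two floating pieces over the copied character pose.
# The event graph computes a time-driven random drift, upper-arm alphas zeroed during skydiving, and an explosive-bow flag.
# Constraints and additive bone offsets use these values plus sprint and parachute states to shift the pieces.
# The cached result feeds a spine-relative RigidBody for output; AnimDynamics and sprint-physics branches sit unconnected.
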AnimBlueprint /BRCosmetics/Characters/Player/Male/Medium/Bodies/M_MED_The_Origin/Meshes/Parts/M_MED_The_Origin_FaceAcc_AnimBP
target_skeleton=&/M_MED_The_Origin_FaceAcc_Skeleton
graphs [
 AnimGraph {nodes [
   n1 = Anim Dynamics - Bone: dyn_floaty_r(Alpha=1, ExternalForce="0,0,0") {
    angularDampingOverride=1
    angularSpringConstant=20000
    bAngularSpring=true
    bGravityOverrideInSimSpace=true
    bLinearSpring=true
    bOverrideAngularDamping=true
    bOverrideLinearDamping=true
    gravityScale=0.35
    linearDampingOverride=1
    linearSpringConstant=120
    simulationSpace=RootRelative
    alphaBoolBlend {blendOption=Linear}
    boundBone {boneName=dyn_floaty_r}
    componentAppliedLinearAccClamp=<3000 3000 3000>
    componentLinearAccScale=<0.6 0.6 0.6>
    componentLinearVelScale=<0.6 0.6 0.6>
    gravityOverride {z=-980}
    physicsBodyDefinitions [{
      boundBone {boneName=dyn_floaty_r}
      boxExtents=<10 10 10>
      constraintSetup {angularLimitsMax=<0 0 0>, angularLimitsMin=<0 0 0>, angularTarget=<0 0 0>}
      localJointOffset=<0 0 0>
     }]
    retargetingSettings {bEnabled=true}
   }
   Anim Dynamics - Bone: dyn_floaty_l(ComponentPose=@n1, Alpha=1, ExternalForce="0,0,0") {
    angularDampingOverride=1
    angularSpringConstant=20000
    bAngularSpring=true
    bGravityOverrideInSimSpace=true
    bLinearSpring=true
    bOverrideAngularDamping=true
    bOverrideLinearDamping=true
    gravityScale=0.35
    linearDampingOverride=1
    linearSpringConstant=120
    simulationSpace=RootRelative
    alphaBoolBlend {blendOption=Linear}
    boundBone {boneName=dyn_floaty_l}
    componentAppliedLinearAccClamp=<3000 3000 3000>
    componentLinearAccScale=<0.6 0.6 0.6>
    componentLinearVelScale=<0.6 0.6 0.6>
    gravityOverride {z=-980}
    physicsBodyDefinitions [{
      boundBone {boneName=dyn_floaty_l}
      boxExtents=<10 10 10>
      constraintSetup {angularLimitsMax=<0 0 0>, angularLimitsMin=<0 0 0>, angularTarget=<0 0 0>}
      localJointOffset=<0 0 0>
     }]
    retargetingSettings {bEnabled=true}
   }
   Blend Poses by bool(BlendTime_0=0.5, BlendTime_1=0.5, bActiveValue=Get bIsSprinting()) {blendType=Linear}
   n2 = RigidBody(ComponentPose=Local To Component(LocalPose=Use cached pose 'prerigid'())) {
    bClampLinearTranslationLimitToRefPose=true
    bOverrideWorldGravity=true
    overridePhysicsAsset=/Game/Characters/Player/Male/Medium/Bodies/M_MED_The_Origin/Meshes/Parts/M_MED_The_Origin_FaceAcc_sprint_Physics
    simulationSpace=BaseBoneSpace
    alphaBoolBlend {blendOption=Linear}
    baseBoneRef {boneName=root}
    componentAppliedLinearAccClamp=<300 300 300>
    componentLinearAccScale=<0.3 0.3 0.3>
    componentLinearVelScale=<0.3 0.3 0.3>
   }
   Component To Local(ComponentPose=@n2)
   "Constraint - dyn_floaty_r_MSTR (P)"(Alpha=1, ConstraintWeights_0=1) {
    alphaBoolBlend {blendOption=Linear}
    boneToModify {boneName=dyn_floaty_r_MSTR}
    constraintSetup [{targetBone {boneName=clavicle_r}}]
    constraintWeights [1]
   }
   Look At - Bone: dyn_spine_05(Alpha=1) {
    alphaBoolBlend {blendOption=Linear}
    boneToModify {boneName=dyn_spine_05}
    lookAtLocation {x=0}
    lookAtTarget {boneReference {boneName=head}}
    lookAt_Axis {axis=<1 0>}
   }
   RigidBody!(ComponentPose, bAlphaBoolEnabled:bool, ExternalForce) -> (Pose) {
    alphaInputType=Bool
    bClampLinearTranslationLimitToRefPose=true
    bOverrideWorldGravity=true
    overridePhysicsAsset=&/M_MED_The_Origin_FaceAcc_Physics
    simulationSpace=BaseBoneSpace
    alphaBoolBlend {blendInTime=0.1, blendOption=Linear, blendOutTime=0.1}
    baseBoneRef {boneName=spine_05}
    componentAppliedLinearAccClamp=<3000 3000 3000>
    componentLinearAccScale=<0.5 0.5 0.5>
    componentLinearVelScale=<0.5 0.5 0.5>
    simSpaceSettings {
     maxAngularAcceleration=200
     maxAngularVelocity=5
     maxLinearAcceleration=1000
     maxLinearVelocity=300
     velocityScaleZ=0
     worldAlpha=1
     externalLinearDragV=<1 1>
    }
   }
   n3 = RigidBody!(ComponentPose=Local To Component(LocalPose=Use cached pose 'prerigid'()), Alpha:float, ExternalForce) -> (Pose) {
    bClampLinearTranslationLimitToRefPose=true
    bOverrideWorldGravity=true
    overridePhysicsAsset=&M_MED_The_Origin_FaceAcc_Physics
    simulationSpace=BaseBoneSpace
    alphaBoolBlend {blendOption=Linear}
    baseBoneRef {boneName=spine_05}
    componentAppliedLinearAccClamp=<300 300 300>
    componentLinearAccScale=<0.5 0.5 0.5>
    componentLinearVelScale=<0.5 0.5 0.5>
   }
   Output Pose(Result=Component To Local(ComponentPose=@n3)) {layerGroup=None}
   n4 = Copy Pose From Mesh(SourceMeshComponent=Get Body_Mesh())
   n5 = Copy Pose From Mesh() {bUseAttachedParent=true}
   n6 = "Constraint - dyn_spine_05 (PP)"!(
    ComponentPose=Local To Component(LocalPose=Layered blend per bone(BasePose=@n5, BlendPoses_0=@n4) {layerSetup [{branchFilters [{boneName=upperarm_l}, {boneName=upperarm_r}]}]})
    Alpha=1
    ConstraintWeights_0=1
    ConstraintWeights_1:float=0
   ) -> (Pose) {
    alphaBoolBlend {blendOption=Linear}
    boneToModify {boneName=dyn_spine_05}
    constraintSetup [{targetBone {boneName=spine_05}}, {targetBone {boneName=neck_02}}]
    constraintWeights [1,1]
   }
   n7 = "Constraint - dyn_spine_05 (PP)"!(ComponentPose=@n6, bAlphaBoolEnabled:bool=Get bow() -> (bow:bool), ConstraintWeights_0=0.5, ConstraintWeights_1=0.5) -> (Pose) {
    alphaInputType=Bool
    alphaBoolBlend {blendInTime=0.1, blendOption=Linear, blendOutTime=0.1}
    boneToModify {boneName=dyn_spine_05}
    constraintSetup [{targetBone {boneName=head}}, {targetBone {boneName=spine_05}}]
    constraintWeights [1,1]
   }
   n8 = Get bIsSprinting()
   n9 = "Transform (Modify) Bone - Bone: dyn_floaty_r_MSTR"!(ComponentPose=@n7, bAlphaBoolEnabled:bool=@n8, Translation="-1.5,1,0", Rotation="0,0,0", Scale="1,1,1") -> (Pose) {
    alphaInputType=Bool
    rotationSpace=BCS_WorldSpace
    translationMode=BMM_Additive
    translationSpace=BCS_BoneSpace
    alphaBoolBlend {blendInTime=0.5, blendOption=Linear, blendOutTime=0.5}
    alphaScaleBiasClamp {interpSpeedDecreasing=2, interpSpeedIncreasing=2}
    boneToModify {boneName=dyn_floaty_r_MSTR}
    translation=<-1.5 1>
   }
   n10 = "Transform (Modify) Bone - Bone: dyn_floaty_l_MSTR"!(ComponentPose=@n9, bAlphaBoolEnabled:bool=@n8, Translation="-1.5,-1,0", Rotation="0,0,0", Scale="1,1,1") -> (Pose) {
    alphaInputType=Bool
    rotationSpace=BCS_WorldSpace
    translationMode=BMM_Additive
    translationSpace=BCS_BoneSpace
    alphaBoolBlend {blendInTime=0.5, blendOption=Linear, blendOutTime=0.5}
    alphaScaleBiasClamp {interpSpeedDecreasing=2, interpSpeedIncreasing=2}
    boneToModify {boneName=dyn_floaty_l_MSTR}
    translation=<-1.5 -1>
   }
   n11 = Get bIsParachuteOpened()
   n12 = "Transform (Modify) Bone - Bone: dyn_floaty_r_MSTR"(ComponentPose=@n10, bAlphaBoolEnabled=@n11, Translation="-4,-1,0", Rotation="0,0,0", Scale="1,1,1") {
    alphaInputType=Bool
    rotationSpace=BCS_WorldSpace
    translationMode=BMM_Additive
    translationSpace=BCS_BoneSpace
    alphaBoolBlend {blendInTime=0.1, blendOption=Linear, blendOutTime=0.1}
    alphaScaleBiasClamp {interpSpeedDecreasing=2, interpSpeedIncreasing=2}
    boneToModify {boneName=dyn_floaty_r_MSTR}
    translation=<-4 -1>
   }
   n13 = "Transform (Modify) Bone - Bone: dyn_floaty_l_MSTR"(ComponentPose=@n12, bAlphaBoolEnabled=@n11, Translation="-4,1,0", Rotation="0,0,0", Scale="1,1,1") {
    alphaInputType=Bool
    rotationSpace=BCS_WorldSpace
    translationMode=BMM_Additive
    translationSpace=BCS_BoneSpace
    alphaBoolBlend {blendInTime=0.1, blendOption=Linear, blendOutTime=0.1}
    alphaScaleBiasClamp {interpSpeedDecreasing=2, interpSpeedIncreasing=2}
    boneToModify {boneName=dyn_floaty_l_MSTR}
    translation=<-4 1>
   }
   n14 = Get float movement() -> (float movement:Vector)
   n15 = "Transform (Modify) Bone - Bone: dyn_floaty_r_MSTR"!(ComponentPose=@n13, Alpha:float=1, Translation=@n14, Rotation="0,0,0", Scale="1,1,1") -> (Pose) {
    rotationSpace=BCS_WorldSpace
    translationMode=BMM_Additive
    translationSpace=BCS_BoneSpace
    alphaBoolBlend {blendOption=Linear}
    alphaScaleBiasClamp {interpSpeedDecreasing=2, interpSpeedIncreasing=2}
    boneToModify {boneName=dyn_floaty_r_MSTR}
   }
   n16 = "Transform (Modify) Bone - Bone: dyn_floaty_l_MSTR"!(ComponentPose=@n15, Alpha:float=1, Translation=@n14, Rotation="0,0,0", Scale="1,1,1") -> (Pose) {
    rotationSpace=BCS_WorldSpace
    translationMode=BMM_Additive
    translationSpace=BCS_BoneSpace
    alphaBoolBlend {blendOption=Linear}
    alphaScaleBiasClamp {interpSpeedDecreasing=2, interpSpeedIncreasing=2}
    boneToModify {boneName=dyn_floaty_l_MSTR}
   }
   n17 = "Transform (Modify) Bone - Bone: dyn_floaty_r_MSTR"(ComponentPose=@n16, Alpha=Get Arm R rot() -> (Arm R rot:double), Translation="-1.671,3,0.486", Rotation="0.000,0.000,-35.000", Scale="1,1,1") {
    rotationMode=BMM_Additive
    rotationSpace=BCS_BoneSpace
    translationMode=BMM_Additive
    translationSpace=BCS_BoneSpace
    alphaBoolBlend {blendOption=Linear}
    alphaScaleBiasClamp {interpSpeedDecreasing=2, interpSpeedIncreasing=2}
    boneToModify {boneName=dyn_floaty_r_MSTR}
    rotation=<rot 0 0 -35>
    translation=<-1.671 3 0.486>
   }
   n18 = "Transform (Modify) Bone - Bone: dyn_floaty_l_MSTR"(ComponentPose=@n17, Alpha=Get Arm L rot() -> (Arm L rot:double), Translation="-7,1.2,-0.000", Rotation="0,0,0", Scale="1,1,1") {
    rotationSpace=BCS_WorldSpace
    translationMode=BMM_Additive
    translationSpace=BCS_BoneSpace
    alphaBoolBlend {blendOption=Linear}
    alphaScaleBiasClamp {interpSpeedDecreasing=2, interpSpeedIncreasing=2}
    boneToModify {boneName=dyn_floaty_l_MSTR}
    translation=<-7 1.2>
   }
   n19 = "Transform (Modify) Bone - Bone: dyn_floaty_r_MSTR"(ComponentPose=@n18, Alpha=Get Arm R rot_2() -> (Arm R rot_2:double), Translation="-4,0,-0.000", Rotation="0,0,0", Scale="1,1,1") {
    rotationMode=BMM_Additive
    rotationSpace=BCS_BoneSpace
    translationMode=BMM_Additive
    translationSpace=BCS_BoneSpace
    alphaBoolBlend {blendOption=Linear}
    boneToModify {boneName=dyn_floaty_r_MSTR}
    translation {x=-4}
   }
   Save cached pose 'prerigid'(Pose=Component To Local(ComponentPose=@n19)) {cacheName=prerigid}
   Get Arm R rot()
  ]}
 EventGraph {nodes [
   Event BlueprintUpdateAnimation()
   > n1 = Sequence!(execute) -> (then_0, then_1, then_2:Exec)
   > n2 = Random Vector Function!(
    execute
    CharacterPartAnimInstance:Custom Character Part Anim Instance Object Reference
    SkelMeshComponent:Mesh Component Object Reference=GetOwnerSkeletalMeshComponent()
    JointName:Name=None
    GameTimeInSeconds:double=GetGameTimeInSeconds()
    TimeOffset:double=0
    PerlinScale:double=0
    Frequency:Vector="1.1,0.3,0.75"
    Amplitude:Vector="1,0.7,0.5"
    Offset:Vector="-2,0,0"
   ) -> (then, VectorOut:Vector)
   > Set float movement(float movement:Vector=@n2) -> (Output_Get:Vector)
   > n3 = Branch(Condition=Get bIsSkyDiving())
   > Set Arm L rot(Arm L rot:double=0) -> (Output_Get:double)
   > Set Arm R rot(Arm R rot:double=0) -> (Output_Get:double)
   n4 = Get Skel_Mesh()
   Set Arm L rot(
    execute=@n3.else
    Arm L rot=GetAlphaFromReferenceTransformDelta(SkeletalMeshComponent=@n4, BoneName=upperarm_l, BaseName=spine_05, InputRange=(X=60,Y=100), TwistAxis="1,0,0")
   )
   n5 = GetAlphaFromReferenceTransformDelta(SkeletalMeshComponent=@n4, BoneName=upperarm_r, BaseName=spine_05, InputRange=(X=60,Y=120), TwistAxis="1,0,0")
   > Set Arm R rot(Arm R rot=MapRangeClamped(Value="Absolute (Float)"(A=@n5.RawOutput), InRangeA=60, InRangeB=120))
   n6 = Cast To FortPlayerAnimInstance(execute=@n1.then_1, Object=GetAnimInstance(self=GetOwnerSkeletalMeshComponent()))
   > Set bow(bow:bool=false) -> (Output_Get:bool)
   > n7 = Switch on EFortWeaponCoreAnimation!(execute, Selection:EFortWeaponCoreAnimation Enum=Get WeaponCoreAnim(self=@n6)) -> (
    Melee:Exec
    Pistol:Exec
    Shotgun:Exec
    PaperBlueprint:Exec
    Rifle:Exec
    MeleeOneHand:Exec
    MachinePistol:Exec
    RocketLauncher:Exec
    GrenadeLauncher:Exec
    GoingCommando:Exec
    AssaultRifle:Exec
    TacticalShotgun:Exec
    SniperRifle:Exec
    TrapPlacement:Exec
    ShoulderLauncher:Exec
    AbilityDecoTool:Exec
    Crossbow:Exec
    C4:Exec
    RemoteControl:Exec
    DualWield:Exec
    AR_BullPup:Exec
    AR_ForwardGrip:Exec
    MedPackPaddles:Exec
    SMG_P90:Exec
    AR_DrumGun:Exec
    Consumable_Small:Exec
    Consumable_Large:Exec
    Balloon:Exec
    MountedTurret:Exec
    CreativeTool:Exec
    ExplosiveBow:Exec
    AshtonIndigo:Exec
    AshtonChicago:Exec
    MeleeDualWield:Exec
    Unarmed:Exec
   )
   Set bow(execute=@n7.ExplosiveBow, bow=true)
   n8 = Branch(execute=@n1.then_2, Condition=Get bow())
   > Set Arm R rot_2(Arm R rot_2:double=GetAlphaFromReferenceTransformDelta(SkeletalMeshComponent=Get Skel_Mesh(), BoneName=upperarm_r, BaseName=spine_05, InputRange=(X=45,Y=100), TwistAxis="1,0,0")) -> (Output_Get:double)
   Set Arm R rot_2(execute=@n8.else, Arm R rot_2=0)
   Event BlueprintInitializeAnimation()
   n9 = GetOwnerSkeletalMeshComponent()
   > Set Skel_Mesh(Skel_Mesh=@n9)
   > Set Body_Mesh(Body_Mesh=@n9)
  ]}
]
variables [
 Arm L rot {type=number}
 Arm R rot {type=number}
 Arm R rot_2 {type=number}
 Body_Mesh {type=SkeletalMeshComponent}
 bow {type=boolean}
 float movement {type=Vector}
 Skel_Mesh {type=SkeletalMeshComponent}
]
